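# Sound cue for drawing a bone bow, routed as weapon foley with reload attenuation and two bow concurrency rules.
# A mixer layers a random pick of three draw one-shots with a quieter looping tension sound.
# The loop starts after a fixed delay and passes through an envelope that fades its volume in.
# Modulators vary the volume of each layer and sit over the whole mix.
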SoundCue /PrimalGameplay/Sounds/Weapons/Bows/Bone/Bow_Bone_Draw_Cue
root=Modulator_0
nodes [
 {id=Delay_0, children [WavePlayer_2], properties {delayMax=0.7, delayMin=0.7}}
 {
  id=Enveloper_0
  children [Delay_0]
  properties {pitchCurve {editorCurveData {keys [{time=0}]}}, volumeCurve {editorCurveData {keys [<at 0 0>,<at 0.5291 0.9862>]}}}
 }
 {id=Mixer_0, children [Modulator_1,Modulator_2], properties {inputVolume [0.8,0.35]}}
 {id=Modulator_0, children [Mixer_0]}
 {id=Modulator_1, children [Random_1], properties {volumeMax=0.8, volumeMin=0.7}}
 {id=Modulator_2, children [Enveloper_0], properties {volumeMax=0.8, volumeMin=0.7}}
 {id=Random_1, children [WavePlayer_4,WavePlayer_5,WavePlayer_0], properties {weights [1,1,1]}}
 {id=WavePlayer_0, properties {soundWaveAssetPtr=&/Bow_Bone_Draw_03}}
 {id=WavePlayer_2, properties {bLooping=true, soundWaveAssetPtr=&/Bow_Bone_Draw_TensionLoop_03}}
 {id=WavePlayer_4, properties {soundWaveAssetPtr=&/Bow_Bone_Draw_01}}
 {id=WavePlayer_5, properties {soundWaveAssetPtr=&/Bow_Bone_Draw_02}}
]
properties {
 attenuationSettings=&Fort_Weapon_Reload_Att
 bDebug=true
 maxDistance=2250
 soundClassObject=&Fort_WeaponFoley
 volumeMultiplier=0.251
 concurrencySet [/Game/Sounds/Concurrency_Rules/Bow_Global_Conc, /Game/Sounds/Concurrency_Rules/Bow_Local_Conc]
}
