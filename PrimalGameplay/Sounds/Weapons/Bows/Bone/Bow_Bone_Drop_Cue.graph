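# Sound cue that plays the drop sound for the bone bow weapon.
# A Random node picks one of three equally weighted Bow_Bone_Drop wave variations.
# The cue slightly lowers pitch and volume, uses weapon drop attenuation and the loot collect sound class, and applies global and local bow concurrency rules.
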SoundCue /PrimalGameplay/Sounds/Weapons/Bows/Bone/Bow_Bone_Drop_Cue
root=Random_0
nodes [
 {id=Random_0, children [WavePlayer_0,WavePlayer_2,WavePlayer_3], properties {weights [1,1,1]}}
 {id=WavePlayer_0, properties {soundWaveAssetPtr=&/Bow_Bone_Drop_01}}
 {id=WavePlayer_2, properties {soundWaveAssetPtr=&/Bow_Bone_Drop_02}}
 {id=WavePlayer_3, properties {soundWaveAssetPtr=&/Bow_Bone_Drop_03}}
]
properties {
 attenuationSettings=&Fort_Weapon_Drop_Att
 maxDistance=3000
 pitchMultiplier=0.9
 soundClassObject=&Fort_Loot_Collect
 volumeMultiplier=0.8
 concurrencySet [/Game/Sounds/Concurrency_Rules/Bow_Global_Conc, /Game/Sounds/Concurrency_Rules/Bow_Local_Conc]
}
