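# Sound cue for the equip foley of the bone bow weapon.
# A random node picks one of three equip wave files with equal weight, and a modulator varies its pitch slightly downward.
# It plays quietly in the weapon foley sound class, using the weapon reload attenuation and both global and local bow concurrency rules.
# The debug flag is left enabled on the cue.
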SoundCue /PrimalGameplay/Sounds/Weapons/Bows/Bone/Bow_Bone_Equip_Cue
root=Modulator_0
nodes [
 {id=Modulator_0, children [Random_0], properties {pitchMax=1, pitchMin=0.9}}
 {id=Random_0, children [WavePlayer_1,WavePlayer_2,WavePlayer_3], properties {weights [1,1,1]}}
 {id=WavePlayer_1, properties {soundWaveAssetPtr=&/Bow_Bone_Equip_01}}
 {id=WavePlayer_2, properties {soundWaveAssetPtr=&/Bow_Bone_Equip_02}}
 {id=WavePlayer_3, properties {soundWaveAssetPtr=&/Bow_Bone_Equip_03}}
]
properties {
 attenuationSettings=&Fort_Weapon_Reload_Att
 bDebug=true
 maxDistance=2250
 soundClassObject=&Fort_WeaponFoley
 volumeMultiplier=0.2
 concurrencySet [/Game/Sounds/Concurrency_Rules/Bow_Global_Conc, /Game/Sounds/Concurrency_Rules/Bow_Local_Conc]
}
