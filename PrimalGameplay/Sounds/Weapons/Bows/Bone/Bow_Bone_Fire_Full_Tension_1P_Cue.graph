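# First-person sound cue for firing a bone bow at full tension.
# A random node picks one of five equally weighted waves, all named as Normal Tension recordings, then a modulator slightly varies pitch and volume downward.
# It routes through a first-person ranged weapon fire sound class, a projectile attenuation asset and first-person gunfire concurrency.
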SoundCue /PrimalGameplay/Sounds/Weapons/Bows/Bone/Bow_Bone_Fire_Full_Tension_1P_Cue
root=Modulator_0
nodes [
 {id=Modulator_0, children [Random_3], properties {pitchMax=1, pitchMin=0.9, volumeMax=1, volumeMin=0.9}}
 {id=Random_3, children [WavePlayer_12, WavePlayer_13, WavePlayer_16, WavePlayer_17, WavePlayer_18], properties {weights [1,1,1,1,1]}}
 {id=WavePlayer_12, properties {soundWaveAssetPtr=&/Bow_Bone_Fire_Normal_Tension_1P_02}}
 {id=WavePlayer_13, properties {soundWaveAssetPtr=&/Bow_Bone_Fire_Normal_Tension_1P_03}}
 {id=WavePlayer_16, properties {soundWaveAssetPtr=&/Bow_Bone_Fire_Normal_Tension_1P_B02}}
 {id=WavePlayer_17, properties {soundWaveAssetPtr=&/Bow_Bone_Fire_Normal_Tension_1P_B03}}
 {id=WavePlayer_18, properties {soundWaveAssetPtr=&/Bow_Bone_Fire_Normal_Tension_1P_B04}}
]
properties {
 attenuationSettings=&Fort_Weapon_Projectile_FirstPerson_Athena_Att
 maxDistance=2097152
 soundClassObject=&Fort_WeaponRanged_Fire_1P_PSM
 volumeMultiplier=0.95
 attenuationOverrides {falloffDistance=5000, attenuationShapeExtents {x=1000}}
 concurrencySet [&Weapon_Gunfire_FirstPerson_Conc]
}
