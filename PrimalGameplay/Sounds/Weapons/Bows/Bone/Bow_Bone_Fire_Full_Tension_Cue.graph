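# Sound cue for firing the bone bow at full tension, blending close and distant layers through separate sound classes and attenuations.
# The close layer randomizes pitch- and volume-modulated Normal Tension 1P waves under a third-person fire class.
# The distant layer mixes random Full Tension Medium waves with a compound bow shot borrowed from Explosive Bow.
# Global and local bow concurrency rules limit simultaneous playback.
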SoundCue /PrimalGameplay/Sounds/Weapons/Bows/Bone/Bow_Bone_Fire_Full_Tension_Cue
root=Modulator_3
nodes [
 {
  id=Attenuation_3
  children [Modulator_4]
  properties {attenuationSettings=/Game/Sounds/Attenuation/Fort_Weapon_Projectile_Close_Bow_Athena_Att}
 }
 {
  id=Attenuation_4
  children [Mixer_0]
  properties {attenuationSettings=/Game/Sounds/Attenuation/Fort_Weapon_Projectile_Distant_Bow_Athena_Att}
 }
 {id=Mixer_0, children [Modulator_1,Modulator_2], properties {inputVolume [0.9,0.9]}}
 {id=Mixer_1, children [SoundClass_3,SoundClass_2], properties {inputVolume [0.8,0.5]}}
 {id=Modulator_1, children [Random_2], properties {pitchMax=1.2, pitchMin=0.9, volumeMax=1.1, volumeMin=0.9}}
 {id=Modulator_2, children [WavePlayer_9], properties {pitchMax=1, pitchMin=0.8, volumeMax=0.8, volumeMin=0.8}}
 {id=Modulator_3, children [Mixer_1]}
 {id=Modulator_4, children [Random_4], properties {pitchMax=1, pitchMin=0.9, volumeMax=1, volumeMin=0.9}}
 {id=Random_2, children [WavePlayer_5,WavePlayer_6,WavePlayer_1,WavePlayer_7], properties {weights [1,1,1,1]}}
 {id=Random_4, children [WavePlayer_17, WavePlayer_18, WavePlayer_19, WavePlayer_20, WavePlayer_21], properties {weights [1,1,1,1,1]}}
 {
  id=SoundClass_2
  children [Attenuation_4]
  properties {soundClassOverride=&Fort_WeaponRanged_Fire_Distant}
 }
 {
  id=SoundClass_3
  children [Attenuation_3]
  properties {soundClassOverride=&Fort_WeaponRanged_Fire_3P_PSM}
 }
 {id=WavePlayer_1, properties {soundWaveAssetPtr=&/Bow_Bone_Fire_Full_Tension_Medium_02}}
 {id=WavePlayer_17, properties {soundWaveAssetPtr=&/Bow_Bone_Fire_Normal_Tension_1P_02}}
 {id=WavePlayer_18, properties {soundWaveAssetPtr=&/Bow_Bone_Fire_Normal_Tension_1P_03}}
 {id=WavePlayer_19, properties {soundWaveAssetPtr=&/Bow_Bone_Fire_Normal_Tension_1P_B02}}
 {id=WavePlayer_20, properties {soundWaveAssetPtr=&/Bow_Bone_Fire_Normal_Tension_1P_B03}}
 {id=WavePlayer_21, properties {soundWaveAssetPtr=&/Bow_Bone_Fire_Normal_Tension_1P_B04}}
 {id=WavePlayer_5, properties {soundWaveAssetPtr=&/Bow_Bone_Fire_Full_Tension_Medium_03}}
 {id=WavePlayer_6, properties {soundWaveAssetPtr=&/Bow_Bone_Fire_Full_Tension_Medium_01}}
 {id=WavePlayer_7, properties {soundWaveAssetPtr=&/Bow_Bone_Fire_Full_Tension_Medium_04}}
 {id=WavePlayer_9, properties {soundWaveAssetPtr=/Game/Sounds/Fort_Weapons/Explosive_Bow/Weapon_Bow_Compound_Shoot_Med_01}}
]
properties {
 maxDistance=27500
 pitchMultiplier=1.1
 volumeMultiplier=1
 attenuationOverrides {falloffDistance=5000, attenuationShapeExtents {x=1000}}
 concurrencySet [/Game/Sounds/Concurrency_Rules/Bow_Global_Conc, /Game/Sounds/Concurrency_Rules/Bow_Local_Conc]
}
