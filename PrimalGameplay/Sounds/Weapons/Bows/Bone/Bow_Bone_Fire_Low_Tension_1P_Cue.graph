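# First-person sound cue for firing a bone bow at low tension.
# A random node picks one of three equally weighted fire recordings, and a modulator varies its pitch slightly downward.
# The cue plays at reduced volume through the first-person ranged-fire sound class, with first-person projectile attenuation and first-person gunfire concurrency.
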SoundCue /PrimalGameplay/Sounds/Weapons/Bows/Bone/Bow_Bone_Fire_Low_Tension_1P_Cue
root=Modulator_0
nodes [
 {id=Modulator_0, children [Random_1], properties {pitchMax=1, pitchMin=0.9}}
 {id=Random_1, children [WavePlayer_4,WavePlayer_5,WavePlayer_6], properties {weights [1,1,1]}}
 {id=WavePlayer_4, properties {soundWaveAssetPtr=&/Bow_Bone_Fire_Low_Tension_1P_01}}
 {id=WavePlayer_5, properties {soundWaveAssetPtr=&/Bow_Bone_Fire_Low_Tension_1P_02}}
 {id=WavePlayer_6, properties {soundWaveAssetPtr=&/Bow_Bone_Fire_Low_Tension_1P_03}}
]
properties {
 attenuationSettings=&Fort_Weapon_Projectile_FirstPerson_Athena_Att
 maxDistance=2097152
 soundClassObject=&Fort_WeaponRanged_Fire_1P_PSM
 volumeMultiplier=0.5
 attenuationOverrides {falloffDistance=5000, attenuationShapeExtents {x=1000}}
 concurrencySet [&Weapon_Gunfire_FirstPerson_Conc]
}
