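# Sound cue for firing the bone bow at low tension, mixing a close layer and a distant layer.
# Close layer randomizes three 1P-named shots under a third-person fire sound class and close bow attenuation.
# Distant layer blends four random medium shots with a compound bow shot from the explosive bow, using distant class and attenuation.
# Modulators vary pitch and volume, and global and local bow concurrency rules limit playback.
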SoundCue /PrimalGameplay/Sounds/Weapons/Bows/Bone/Bow_Bone_Fire_Low_Tension_Cue
root=Modulator_3
nodes [
 {
  id=Attenuation_3
  children [Modulator_2]
  properties {attenuationSettings=/Game/Sounds/Attenuation/Fort_Weapon_Projectile_Close_Bow_Athena_Att}
 }
 {
  id=Attenuation_4
  children [Mixer_0]
  properties {attenuationSettings=/Game/Sounds/Attenuation/Fort_Weapon_Projectile_Distant_Bow_Athena_Att}
 }
 {id=Mixer_0, children [Modulator_0,Modulator_1], properties {inputVolume [1,0.6]}}
 {id=Mixer_1, children [SoundClass_3,SoundClass_2], properties {inputVolume [0.6,0.5]}}
 {id=Modulator_0, children [Random_2], properties {pitchMax=1.1, pitchMin=0.9, volumeMax=0.9, volumeMin=0.7}}
 {id=Modulator_1, children [WavePlayer_8], properties {pitchMax=1, pitchMin=0.88, volumeMax=0.8, volumeMin=0.7}}
 {id=Modulator_2, children [Random_1], properties {pitchMax=1.1, pitchMin=0.9}}
 {id=Modulator_3, children [Mixer_1]}
 {id=Random_1, children [WavePlayer_2, WavePlayer_10, WavePlayer_11], properties {weights [1,1,1]}}
 {id=Random_2, children [WavePlayer_6,WavePlayer_7,WavePlayer_9,WavePlayer_1], properties {weights [1,1,1,1]}}
 {
  id=SoundClass_2
  children [Attenuation_4]
  properties {soundClassOverride=&Fort_WeaponRanged_Fire_Distant}
 }
 {
  id=SoundClass_3
  children [Attenuation_3]
  properties {soundClassOverride=&Fort_WeaponRanged_Fire_3P_PSM}
 }
 {id=WavePlayer_1, properties {soundWaveAssetPtr=&/Bow_Bone_Fire_Low_Tension_Medium_04}}
 {id=WavePlayer_10, properties {soundWaveAssetPtr=&/Bow_Bone_Fire_Low_Tension_1P_02}}
 {id=WavePlayer_11, properties {soundWaveAssetPtr=&/Bow_Bone_Fire_Low_Tension_1P_03}}
 {id=WavePlayer_2, properties {soundWaveAssetPtr=&/Bow_Bone_Fire_Low_Tension_1P_01}}
 {id=WavePlayer_6, properties {soundWaveAssetPtr=&/Bow_Bone_Fire_Low_Tension_Medium_01}}
 {id=WavePlayer_7, properties {soundWaveAssetPtr=&/Bow_Bone_Fire_Low_Tension_Medium_02}}
 {id=WavePlayer_8, properties {soundWaveAssetPtr=/Game/Sounds/Fort_Weapons/Explosive_Bow/Weapon_Bow_Compound_Shoot_Med_01}}
 {id=WavePlayer_9, properties {soundWaveAssetPtr=&/Bow_Bone_Fire_Low_Tension_Medium_03}}
]
properties {
 maxDistance=27500
 volumeMultiplier=0.9
 attenuationOverrides {falloffDistance=5000, attenuationShapeExtents {x=1000}}
 concurrencySet [/Game/Sounds/Concurrency_Rules/Bow_Global_Conc, /Game/Sounds/Concurrency_Rules/Bow_Local_Conc]
}
