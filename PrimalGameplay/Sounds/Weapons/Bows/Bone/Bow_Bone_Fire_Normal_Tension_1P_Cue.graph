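# Sound cue for the first-person bone bow firing sound at normal tension.
# An equally weighted Random node picks one of three low-tension fire waves, despite the cue's normal-tension name.
# A root Modulator varies volume and lowers pitch by a random amount, on top of reduced cue-level pitch and volume.
# It uses first-person projectile attenuation, a ranged-weapon fire sound class and first-person gunfire concurrency.
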SoundCue /PrimalGameplay/Sounds/Weapons/Bows/Bone/Bow_Bone_Fire_Normal_Tension_1P_Cue
root=Modulator_0
nodes [
 {id=Modulator_0, children [Random_2], properties {pitchMax=0.95, pitchMin=0.75, volumeMax=1.1, volumeMin=0.9}}
 {id=Random_2, children [WavePlayer_6,WavePlayer_8,WavePlayer_9], properties {weights [1,1,1]}}
 {id=WavePlayer_6, properties {soundWaveAssetPtr=&/Bow_Bone_Fire_Low_Tension_1P_01}}
 {id=WavePlayer_8, properties {soundWaveAssetPtr=&/Bow_Bone_Fire_Low_Tension_1P_02}}
 {id=WavePlayer_9, properties {soundWaveAssetPtr=&/Bow_Bone_Fire_Low_Tension_1P_03}}
]
properties {
 attenuationSettings=&Fort_Weapon_Projectile_FirstPerson_Athena_Att
 maxDistance=2097152
 pitchMultiplier=0.9
 soundClassObject=&Fort_WeaponRanged_Fire_1P_PSM
 volumeMultiplier=0.5
 attenuationOverrides {falloffDistance=5000, attenuationShapeExtents {x=1000}}
 concurrencySet [&Weapon_Gunfire_FirstPerson_Conc]
}
