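# Sound cue for firing a bone bow at normal tension, layering close and distant shot sounds.
# The close layer plays a random low-tension 1P sample through close attenuation and a third-person fire sound class.
# The distant layer mixes a random normal-tension medium sample with a compound-bow shot through distant attenuation and its own sound class.
# Modulators vary pitch and volume, and concurrency uses global and local bow rules.
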SoundCue /PrimalGameplay/Sounds/Weapons/Bows/Bone/Bow_Bone_Fire_Normal_Tension_Cue
root=Modulator_3
nodes [
 {
  id=Attenuation_3
  children [Modulator_4]
  properties {attenuationSettings=/Game/Sounds/Attenuation/Fort_Weapon_Projectile_Close_Bow_Athena_Att}
 }
 {
  id=Attenuation_4
  children [Mixer_0]
  properties {attenuationSettings=/Game/Sounds/Attenuation/Fort_Weapon_Projectile_Distant_Bow_Athena_Att}
 }
 {id=Mixer_0, children [Modulator_2,Modulator_1], properties {inputVolume [0.9,0.7]}}
 {id=Mixer_1, children [SoundClass_3,SoundClass_2], properties {inputVolume [0.7,0.5]}}
 {id=Modulator_1, children [WavePlayer_0], properties {pitchMax=1, pitchMin=0.8, volumeMax=0.9, volumeMin=0.7}}
 {id=Modulator_2, children [Random_2], properties {pitchMax=1.1, pitchMin=0.9, volumeMax=1.1, volumeMin=0.9}}
 {id=Modulator_3, children [Mixer_1]}
 {id=Modulator_4, children [Random_1], properties {pitchMax=0.95, pitchMin=0.75, volumeMax=1.1, volumeMin=0.9}}
 {id=Random_1, children [WavePlayer_1,WavePlayer_2,WavePlayer_3], properties {weights [1,1,1]}}
 {id=Random_2, children [WavePlayer_9, WavePlayer_10, WavePlayer_11, WavePlayer_4], properties {weights [1,1,1,1]}}
 {
  id=SoundClass_2
  children [Attenuation_4]
  properties {soundClassOverride=&Fort_WeaponRanged_Fire_Distant}
 }
 {
  id=SoundClass_3
  children [Attenuation_3]
  properties {soundClassOverride=&Fort_WeaponRanged_Fire_3P_PSM}
 }
 {id=WavePlayer_0, properties {soundWaveAssetPtr=/Game/Sounds/Fort_Weapons/Explosive_Bow/Weapon_Bow_Compound_Shoot_Med_01}}
 {id=WavePlayer_1, properties {soundWaveAssetPtr=&/Bow_Bone_Fire_Low_Tension_1P_01}}
 {id=WavePlayer_10, properties {soundWaveAssetPtr=&/Bow_Bone_Fire_Normal_Tension_Medium_02}}
 {id=WavePlayer_11, properties {soundWaveAssetPtr=&/Bow_Bone_Fire_Normal_Tension_Medium_03}}
 {id=WavePlayer_2, properties {soundWaveAssetPtr=&/Bow_Bone_Fire_Low_Tension_1P_02}}
 {id=WavePlayer_3, properties {soundWaveAssetPtr=&/Bow_Bone_Fire_Low_Tension_1P_03}}
 {id=WavePlayer_4, properties {soundWaveAssetPtr=&/Bow_Bone_Fire_Normal_Tension_Medium_04}}
 {id=WavePlayer_9, properties {soundWaveAssetPtr=&/Bow_Bone_Fire_Normal_Tension_Medium_01}}
]
properties {
 maxDistance=27500
 pitchMultiplier=0.9
 volumeMultiplier=1.2
 attenuationOverrides {falloffDistance=5000, attenuationShapeExtents {x=1000}}
 concurrencySet [/Game/Sounds/Concurrency_Rules/Bow_Global_Conc, /Game/Sounds/Concurrency_Rules/Bow_Local_Conc]
}
